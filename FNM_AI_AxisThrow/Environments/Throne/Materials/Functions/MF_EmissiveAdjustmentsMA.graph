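# Material function that applies emissive adjustments to a Material Attributes input, for use in the Throne environment materials.
# It extracts Emissive Color, runs it through MF_FX_EmissiveAdjustments and writes the result back into the attribute set.
# The "Use EyeAdaptation" static switch, described as auto exposure compensation, chooses between the adjusted attributes and a reroute pass-through.
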MaterialFunction /FNM_AI_AxisThrow/Environments/Throne/Materials/Functions/MF_EmissiveAdjustmentsMA
graph {nodes [
  n1 = GetMaterialAttributes(None=NamedRerouteDeclaration(None=FunctionInput() {inputName=InMA, inputType=FunctionInput_MaterialAttributes}) {name=inMA, nodeColor=<rgba 1 0 0.4687 1>}) {attributeGetTypes [#1]}
  n2 = MaterialFunctionCall(Emissive=@n1.EmissiveColor) {materialFunction=&MF_FX_EmissiveAdjustments}
  FunctionOutput(None=StaticSwitchParameter(
    True=SetMaterialAttributes!(MaterialAttributes=@n1.MaterialAttributes, EmissiveColor=@n2.Result) {attributeSetTypes [#1]}
    False=NamedRerouteUsage()
   ) {desc="Enable/Disable EyeAdaptation (auto exposure) compensation", group=Options, parameterName=Use EyeAdaptation})
 ]}
