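# Material function for the Throne environment that blends curve-atlas colors for thorns and legs (vines) into incoming Material Attributes.
# UV windows or curve alpha produce the masks, which drive base color and emissive; AO and normal are also blended toward flat values.
# Static switches let thorns inherit leg color and glow, with lookup via vertex color, UV1 or stepped UV remapping.
# It also outputs the masks, inherited color and leg UV parameters.
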MaterialFunction /FNM_AI_AxisThrow/Environments/Throne/Materials/Functions/MF_QTAtlasBlend
graph {nodes [
  n1 = GetMaterialAttributes(None=FunctionInput() {inputName=MA, inputType=FunctionInput_MaterialAttributes}) {attributeGetTypes [
    #1
    #2
    #3
    #4
   ]}
  n2 = NamedRerouteUsage()
  n3 = LinearInterpolate(A=@n1.BaseColor, B=NamedRerouteUsage(), Alpha=NamedRerouteUsage())
  n4 = LinearInterpolate(A=@n3, B=NamedRerouteUsage(), Alpha=Multiply(A=NamedRerouteUsage(), B=NamedRerouteUsage()))
  n5 = MaterialFunctionCall!(InLegsUV=NamedRerouteUsage(), InLegsUVParams=NamedRerouteUsage(), InUsesCurvesAlpha=NamedRerouteUsage()) {materialFunction=&/MF_QTGetVinesColor}
  n6 = CurveAtlasRowParameter(CurveTime=@n5.LegsCurvetime) {
   atlas=/FNM_AI_AxisThrow/Environments/Throne/Materials/Functions/Curves/T_QTCurveAtlas
   curve=/FNM_AI_AxisThrow/Environments/Throne/Materials/Functions/Curves/Albedo_Curve_02
   defaultValue=1
   group=C_AtlasColorBlend
   parameterName=LK_LegsCurve
  }
  n7 = ComponentMask(None=@n6) {b=true, g=true, r=true}
  n8 = NamedRerouteUsage()
  n9 = StaticSwitch(True=Multiply(A=@n6.A, B=Ceil(None=@n8)), False=@n8, Value=NamedRerouteUsage())
  n10 = SetMaterialAttributes!(
   MaterialAttributes=@n1.MaterialAttributes
   BaseColor=LinearInterpolate(
    A=StaticSwitchParameter(True=@n4, False=@n3) {group=C_AtlasColorBlend, parameterName=LK_ThornsInheritFromLegs, sortPriority=255}
    B=NamedRerouteDeclaration(None=@n7) {name=outLegsColor, nodeColor=<rgba 1 0.6563 0 1>}
    Alpha=NamedRerouteDeclaration(None=@n9) {name=legsParam, nodeColor=<rgba 0 0.7891 1 1>}
   )
   AmbientOcclusion=LinearInterpolate(A=@n1.AmbientOcclusion, B=Constant() {r=1}, Alpha=@n2)
   Normal=LinearInterpolate(A=@n1.Normal, B=Constant3Vector() {constant=<rgba 0 0 1 0>}, Alpha=@n2)
   EmissiveColor=NamedRerouteUsage()
  ) {attributeSetTypes [
    #1
    #2
    #3
    #4
   ]}
  FunctionOutput(None=@n10)
  FunctionOutput(None=NamedRerouteUsage()) {outputName=ThornsParam, sortPriority=1}
  FunctionOutput(None=NamedRerouteUsage()) {outputName=LegsParam, sortPriority=1}
  FunctionOutput(None=NamedRerouteUsage()) {outputName=ThornsCeil, sortPriority=2}
  FunctionOutput(None=NamedRerouteUsage()) {outputName=LegsCeil, sortPriority=3}
  FunctionOutput(None=StaticSwitchParameter(True=NamedRerouteUsage(), False=Constant()) {group=C_AtlasColorBlend, parameterName=LK_ThornsInheritFromLegs, sortPriority=255}) {outputName=ThornsLegsMask, sortPriority=4}
  n11 = NamedRerouteUsage()
  n12 = Multiply(A=Multiply(A=@n11, B=NamedRerouteUsage()), B=Ceil(None=@n11))
  n13 = NamedRerouteUsage()
  n14 = Multiply(A=Multiply(A=NamedRerouteUsage(), B=NamedRerouteUsage()), B=@n13)
  n15 = Constant()
  FunctionOutput(None=StaticSwitchParameter(True=Add(A=@n12, B=@n14), False=@n15) {group=C_AtlasColorBlend, parameterName=LK_ThornsInheritFromLegs, sortPriority=255}) {outputName=ThornsLegsColor, sortPriority=5}
  FunctionOutput(None=StaticSwitchParameter(True=@n13, False=@n15) {group=C_AtlasColorBlend, parameterName=LK_ThornsInheritFromLegs, sortPriority=255}) {outputName=ThornsLegsParam, sortPriority=6}
  FunctionOutput(None=NamedRerouteUsage()) {outputName=LegsUVParams, sortPriority=7}
  n16 = Constant()
  n17 = Constant() {r=1}
  MaterialFunctionCall(X=@n16, Y=@n17, Z=@n16, A=@n17) {materialFunction=&MakeFloat4}
  NamedRerouteDeclaration(None=@n1.EmissiveColor) {name=inEmissive, nodeColor=<rgba 0.1484 1 0 1>}
  n18 = MaterialFunctionCall(Float4=NamedRerouteUsage()) {materialFunction=&BreakOutFloat4Components}
  n19 = Constant()
  n20 = MaterialFunctionCall(Float2=NamedRerouteDeclaration(None=MaterialFunctionCall!(InUV) {materialFunction=&/MF_QTFlipV}.Result) {name=invUV, nodeColor=<rgba 0.7109 1 0 1>}) {materialFunction=&BreakOutFloat2Components}
  n21 = Constant() {r=1}
  n22 = If(A=@n20.R, B=@n18.R, "A > B"=@n21, "A == B"=@n21, "A < B"=@n19)
  n23 = If(A=@n20.R, B=@n18.G, "A > B"=@n19, "A == B"=@n21, "A < B"=@n21)
  n24 = StaticSwitchParameter(
   True=NamedRerouteUsage()
   False=MaterialFunctionCall(Input=NamedRerouteDeclaration(None=@n20.G) {name=invVcoord, nodeColor=<rgba 0 0.625 1 1>}, Input Low=@n18.B, Input High=@n18.A) {materialFunction=&RemapValueRange}.Result
  ) {
   desc=Uses curves Alpha in place of UV Parameters for legs and thorns
   group=C_AtlasColorBlend
   parameterName=LK_UseCurvesAlpha
   sortPriority=255
  }
  n25 = Multiply(A=NamedRerouteDeclaration(None=Multiply(A=@n22, B=@n23)) {name=thornsUParam, nodeColor=<rgba 1 0.0234 0 1>}, B=Saturate(None=@n24))
  n26 = CurveAtlasRowParameter(CurveTime=@n25) {
   atlas=&T_QTCurveAtlas
   curve=/FNM_AI_AxisThrow/Environments/Throne/Materials/Functions/Curves/Albedo_Curve_01
   group=C_AtlasColorBlend
   parameterName=LK_ThornsCurve
   sortPriority=16
  }
  n27 = StaticSwitchParameter(True=@n26.A, False=@n25) {
   desc=Uses curves Alpha in place of UV Parameters for legs and thorns
   group=C_AtlasColorBlend
   parameterName=LK_UseCurvesAlpha
   sortPriority=255
  }
  NamedRerouteDeclaration(None=@n27) {name=thornsParam, nodeColor=<rgba 1 0 0.4687 1>}
  n28 = StaticSwitchParameter(
   True=Constant4Vector() {constant=<rgba 0 1 0 1>}.RGBA
   False=MaterialFunctionCall(
    X=ScalarParameter() {group=C_AtlasColorBlend, parameterName=Legs MinU, sortPriority=33}
    Y=ScalarParameter() {defaultValue=1, group=C_AtlasColorBlend, parameterName=Legs MaxU, sortPriority=34}
    Z=ScalarParameter() {group=C_AtlasColorBlend, parameterName=Legs MinV, sortPriority=35}
    A=ScalarParameter() {defaultValue=1, group=C_AtlasColorBlend, parameterName=Legs MaxV, sortPriority=36}
   ) {materialFunction=&MakeFloat4}.Result
  ) {
   desc=Uses curves Alpha in place of UV Parameters for legs and thorns
   group=C_AtlasColorBlend
   parameterName=LK_UseCurvesAlpha
   sortPriority=255
  }
  NamedRerouteDeclaration(None=@n28) {name=legsUVParams, nodeColor=<rgba 1 0.6563 0 1>}
  n29 = NamedRerouteUsage()
  n30 = VertexColor()
  n31 = TextureCoordinate() {coordinateIndex=1}
  n32 = NamedRerouteUsage()
  n33 = ScalarParameter() {defaultValue=1, group=C_AtlasColorBlend, parameterName=Thorns InheritOffset, sortPriority=36}
  n34 = ComponentMask(None=TextureCoordinate()) {r=true}
  n35 = NamedRerouteUsage()
  n36 = ComponentMask(None=@n35) {r=true}
  n37 = ComponentMask(None=@n35) {g=true}
  n38 = Subtract(A=@n37, B=@n36)
  n39 = NamedRerouteUsage()
  n40 = Add(
   A=Multiply(A=Divide(B=@n32), B=Floor(None=@n33))
   B=MaterialFunctionCall!(
    Input=Subtract(A=@n34, B=@n36)
    Input Low=@n36
    Input High=@n38
    Target Low=ComponentMask(None=@n39) {a=true}
    Target High=ComponentMask(None=@n39) {b=true}
    Steps=@n38
   ) {materialFunction=&/RemapValueRangeStepped}.Smooth
  )
  n41 = StaticSwitch(True=OneMinus(None=ComponentMask(None=@n31) {g=true}), False=OneMinus(None=Saturate(None=@n40)), Value=NamedRerouteUsage())
  n42 = NamedRerouteUsage()
  n43 = MaterialFunctionCall(
   InLegsUV=AppendVector(A=ComponentMask(None=@n39) {r=true}, B=StaticSwitch(True=@n30.G, False=@n41, Value=@n42))
   InLegsUVParams=StaticSwitch(True=NamedRerouteUsage(), False=NamedRerouteUsage(), Value=NamedRerouteUsage())
   InUsesCurvesAlpha=NamedRerouteUsage()
  ) {materialFunction=&MF_QTGetVinesColor}
  n44 = NamedRerouteUsage()
  n45 = Divide(A=Floor(None=Multiply(A=@n44, B=@n43.LegsCurvetime)), B=@n44)
  n46 = NamedRerouteUsage()
  n47 = CurveAtlasRowParameter(CurveTime=NamedRerouteDeclaration(None=StaticSwitch(True=@n43.LegsCurvetime, False=@n45, Value=@n46)) {name=legsThornsCurveTime, nodeColor=<rgba 0 1 0.4844 1>}) {
   atlas=&T_QTCurveAtlas
   curve=&Albedo_Curve_02
   defaultValue=1
   group=C_AtlasColorBlend
   parameterName=LK_LegsCurve
  }
  n48 = StaticSwitch(True=@n47.A, False=@n43.LegsParam, Value=NamedRerouteUsage())
  n49 = Floor(None=Multiply(A=@n29, B=@n48))
  NamedRerouteDeclaration(None=StaticSwitch(True=@n48, False=Divide(A=@n49, B=@n29), Value=NamedRerouteUsage())) {name=thornsLegsParam, nodeColor=<rgba 1 0 0.3047 1>}
  n50 = Multiply(A=NamedRerouteUsage(), B=NamedRerouteUsage())
  n51 = Subtract(A=ComponentMask(None=TextureCoordinate()) {r=true}, B=ComponentMask(None=NamedRerouteUsage()) {r=true})
  n52 = Multiply(A=OneMinus(None=NamedRerouteDeclaration(None=@n50) {name=thornsFullVParam, nodeColor=<rgba 0 0.2969 1 1>}), B=Saturate(None=Ceil(None=@n51)))
  n53 = Ceil(None=NamedRerouteUsage())
  n54 = Power(
   Base=Multiply(A=Saturate(None=@n52), B=Saturate(None=@n53))
   Exp=ScalarParameter() {defaultValue=1, group=C_AtlasColorBlend, parameterName=TH_Inherit MaskContrast, sortPriority=48}
  )
  n55 = NamedRerouteUsage()
  NamedRerouteDeclaration(None=Multiply(A=Saturate(None=@n54), B=OneMinus(None=@n55))) {name=thornsLegsMask, nodeColor=<rgba 1 0.7869 0.0518 1>}
  NamedRerouteDeclaration(None=Subtract(A=@n37, B=Multiply(A=@n36) {constB=2})) {name=Uquant, nodeColor=<rgba 0 1 0.7656 1>}
  NamedRerouteDeclaration(None=@n47) {name=outLegsThornsColor, nodeColor=<rgba 0 1 0.7891 1>}
  n56 = CurveAtlasRowParameter(CurveTime=NamedRerouteUsage()) {
   atlas=&T_QTCurveAtlas
   curve=&Albedo_Curve_01
   group=C_AtlasColorBlend
   parameterName=LK_ThornsCurve
   sortPriority=16
  }
  n57 = StaticSwitchParameter(True=NamedRerouteUsage(), False=ComponentMask(None=@n56) {b=true, g=true, r=true}) {
   desc=Uses curves Alpha in place of UV Parameters for legs and thorns
   group=C_AtlasColorBlend
   parameterName=LK_UseCurvesAlpha
   sortPriority=255
  }
  NamedRerouteDeclaration(None=@n57) {name=thornsColor, nodeColor=<rgba 0.7344 0 1 1>}
  n58 = CurveAtlasRowParameter(CurveTime=NamedRerouteUsage()) {
   atlas=&T_QTCurveAtlas
   curve=&Albedo_Curve_02
   defaultValue=1
   group=C_AtlasColorBlend
   parameterName=LK_LegsEmissiveCurve
  }
  n59 = StaticSwitchParameter(True=@n58, False=NamedRerouteUsage()) {desc=Use AOSR instead, group=A_Layer_1, parameterName=L1_UseEmissiveGradient, sortPriority=255}
  n60 = NamedRerouteUsage()
  n61 = ScalarParameter() {defaultValue=1, group=C_AtlasColorBlend, parameterName=TH_Inherit EmissiveGradientContrast, sortPriority=48}
  n62 = Multiply(A=Multiply(A=@n59, B=Power(Base=@n60, Exp=@n61) {constExponent=2.5}), B=NamedRerouteUsage())
  n63 = Multiply(
   A=Power(Base=@n62, Exp=ScalarParameter() {defaultValue=1, group=C_AtlasColorBlend, parameterName=LK_EmissiveContrast, sortPriority=50})
   B=ScalarParameter() {defaultValue=1, group=C_AtlasColorBlend, parameterName=LK_LegsEmissive, sortPriority=48}
  )
  NamedRerouteDeclaration(None=Multiply(
    A=StaticSwitchParameter(True=@n63, False=@n62) {group=C_AtlasColorBlend, parameterName=LK_ModulateEmissive, sortPriority=255}
    B=ScalarParameter() {defaultValue=1, group=C_AtlasColorBlend, parameterName=TH_Inherit Emissive, sortPriority=48}
   )) {name=thornsLegsAddEmissive, nodeColor=<rgba 1 0 0.5625 1>}
  n64 = Constant() {r=1}
  n65 = ScalarParameter() {defaultValue=1, group=C_AtlasColorBlend, parameterName=Thorns MaxU, sortPriority=25}
  n66 = StaticSwitchParameter(
   True=MaterialFunctionCall(X=@n64, Y=@n65, Z=Constant(), A=@n64) {materialFunction=&MakeFloat4}.Result
   False=MaterialFunctionCall(
    X=ScalarParameter() {group=C_AtlasColorBlend, parameterName=Thorns MinU, sortPriority=24}
    Y=@n65
    Z=ScalarParameter() {group=C_AtlasColorBlend, parameterName=Thorns MinV, sortPriority=26}
    A=ScalarParameter() {defaultValue=1, group=C_AtlasColorBlend, parameterName=Thorns MaxV, sortPriority=27}
   ) {materialFunction=&MakeFloat4}.Result
  ) {
   desc=Uses curves Alpha in place of UV Parameters for legs and thorns
   group=C_AtlasColorBlend
   parameterName=LK_UseCurvesAlpha
   sortPriority=255
  }
  NamedRerouteDeclaration(None=@n66) {name=thornsUVParams, nodeColor=<rgba 0 1 0.1094 1>}
  NamedRerouteDeclaration(None=ComponentMask(None=@n26) {b=true, g=true, r=true}) {name=CAthornsColor, nodeColor=<rgba 0 0.5312 1 1>}
  n67 = StaticSwitchParameter(True=StaticBool() {value=true}, False=StaticBool()) {
   desc=Uses curves Alpha in place of UV Parameters for legs and thorns
   group=C_AtlasColorBlend
   parameterName=LK_UseCurvesAlpha
   sortPriority=255
  }
  NamedRerouteDeclaration(None=@n67) {name=useCurveAlphaBool, nodeColor=<rgba 0 1 0.7656 1>}
  n68 = StaticSwitchParameter(True=StaticBool() {value=true}, False=StaticBool()) {desc=Uses vertex color to lookup Vines UV from Thorns, group=C_AtlasColorBlend, parameterName=LK_UseVertexColor, sortPriority=255}
  NamedRerouteDeclaration(None=@n68) {name=useVtxColorBool, nodeColor=<rgba 1 0.1406 0 1>}
  NamedRerouteDeclaration(None=@n5.LegsParam) {name=legsUVParam, nodeColor=<rgba 0.3594 1 0 1>}
  n69 = StaticSwitchParameter(True=StaticBool() {value=true}, False=StaticBool()) {desc=Uses vertex color to lookup Vines UV from Thorns, group=C_AtlasColorBlend, parameterName=LK_UseInheritUV1, sortPriority=255}
  NamedRerouteDeclaration(None=@n69) {name=useInheritUV1, nodeColor=<rgba 1 0 0.0469 1>}
  n70 = NamedRerouteUsage()
  n71 = Power(
   Base=NamedRerouteUsage()
   Exp=ScalarParameter() {defaultValue=1, group=C_AtlasColorBlend, parameterName=LK_EmissiveContrast, sortPriority=50}
  )
  n72 = ScalarParameter() {defaultValue=1, group=C_AtlasColorBlend, parameterName=LK_ThornsEmissive, sortPriority=48}
  n73 = StaticSwitchParameter(True=Multiply(A=NamedRerouteDeclaration(None=@n71) {name=thornsMask, nodeColor=<rgba 1 0.9844 0 1>}, B=@n72), False=@n72) {group=C_AtlasColorBlend, parameterName=LK_ModulateEmissive, sortPriority=255}
  n74 = NamedRerouteUsage()
  n75 = Ceil(None=NamedRerouteUsage())
  n76 = Power(
   Base=NamedRerouteUsage()
   Exp=ScalarParameter() {defaultValue=1, group=C_AtlasColorBlend, parameterName=LK_EmissiveContrast, sortPriority=50}
  )
  n77 = ScalarParameter() {defaultValue=1, group=C_AtlasColorBlend, parameterName=LK_LegsEmissive, sortPriority=48}
  n78 = StaticSwitchParameter(True=Multiply(A=NamedRerouteDeclaration(None=@n76) {name=legsMask, nodeColor=<rgba 1 0.5775 0.0177 1>}, B=@n77), False=@n77) {group=C_AtlasColorBlend, parameterName=LK_ModulateEmissive, sortPriority=255}
  n79 = NamedRerouteUsage()
  n80 = NamedRerouteDeclaration(None=Ceil(None=NamedRerouteUsage())) {name=legsCeil, nodeColor=<rgba 0.7578 1 0 1>}
  n81 = LinearInterpolate(
   A=LinearInterpolate(A=NamedRerouteUsage(), B=Multiply(A=@n73, B=@n74), Alpha=@n75)
   B=NamedRerouteDeclaration(None=Multiply(A=@n78, B=@n79)) {name=outLegsEmissive, nodeColor=<rgba 0.3359 0 1 1>}
   Alpha=@n80
  )
  n82 = Add(A=NamedRerouteDeclaration(None=@n75) {name=thornsCeil, nodeColor=<rgba 0 0.4375 1 1>}, B=@n80)
  n83 = StaticSwitchParameter(True=Multiply(A=@n81, B=Saturate(None=@n82)), False=@n81) {group=C_AtlasColorBlend, parameterName=LK_ModulateEmissive, sortPriority=255}
  NamedRerouteDeclaration(None=StaticSwitchParameter(True=Add(A=@n70, B=@n83), False=@n83) {group=C_AtlasColorBlend, parameterName=LK_ThornsInheritFromLegs, sortPriority=255}) {name=outEmissive, nodeColor=<rgba 1 0 0.7969 1>}
  n84 = NamedRerouteUsage()
  n85 = NamedRerouteUsage()
  NamedRerouteDeclaration(None=StaticSwitch(True=@n84, False=StaticSwitch(True=@n85, False=@n85, Value=@n85), Value=@n84)) {name=inheritVtxOrUV1, nodeColor=<rgba 0 0.3672 1 1>}
  NamedRerouteDeclaration(None=MaterialFunctionCall(
    X=ScalarParameter() {group=C_AtlasColorBlend, parameterName=Legs MinU, sortPriority=33}
    Y=ScalarParameter() {defaultValue=1, group=C_AtlasColorBlend, parameterName=Legs MaxU, sortPriority=34}
    Z=ScalarParameter() {group=C_AtlasColorBlend, parameterName=LegsInherit MinV, sortPriority=35}
    A=ScalarParameter() {defaultValue=1, group=C_AtlasColorBlend, parameterName=LegsInherit MaxV, sortPriority=36}
   ) {materialFunction=&MakeFloat4}.Result) {name=legsInheritUVParams, nodeColor=<rgba 0.0303 1 0.9537 1>}
  n86 = NamedRerouteUsage()
  n87 = Multiply(A=NamedRerouteUsage(), B=@n86)
  n88 = NamedRerouteUsage()
  n89 = FunctionInput() {bUsePreviewValueAsDefault=true, inputName=EmissiveColor, sortPriority=1, previewValue=<1 1 1 1>}
  n90 = Multiply(
   A=ScalarParameter() {defaultValue=1, group=C_AtlasColorBlend, parameterName=LK_MaskAddBoost, sortPriority=50}
   B=Multiply(A=Add(A=@n88, B=@n86), B=NamedRerouteDeclaration(None=@n89) {name=inEmissivecolor, nodeColor=<rgba 0.7812 1 0 1>})
  )
  NamedRerouteDeclaration(None=StaticSwitchParameter(True=Add(A=@n87, B=@n90), False=NamedRerouteUsage()) {group=C_AtlasColorBlend, parameterName=LK_AddMask, sortPriority=255}) {name=maskedInEmissive, nodeColor=<rgba 0 1 0.3438 1>}
  n91 = ScalarParameter() {defaultValue=1, group=C_AtlasColorBlend, parameterName=LK_MaskAddBoost, sortPriority=50}
  StaticSwitchParameter(True=Add(A=Multiply(A=@n91, B=@n59), B=@n59), False=@n59) {group=C_AtlasColorBlend, parameterName=LK_AddMask, sortPriority=255}
 ]}
