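# Debug material function for thorns-and-legs materials in the Throne environment, showing internal masks and parameters on the surface.
# A ShowDebug static switch passes material attributes through or sets base color to black and drives emissive with the selected view times a multiplier.
# Nested static switches select gradient, fresnel, thorns/legs mask, color or gradient, vine animation, or vertex color views.
# Tint parameters color-code mask inputs, and the thorns/legs gradient supports range remapping.
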MaterialFunction /FNM_AI_AxisThrow/Environments/Throne/Materials/Functions/MF_QT_Debug
graph {nodes [
  n1 = FunctionInput() {inputName=InMA, inputType=FunctionInput_MaterialAttributes}
  n2 = Constant3Vector()
  n3 = NamedRerouteUsage()
  n4 = NamedRerouteUsage()
  n5 = StaticSwitchParameter(
   True=NamedRerouteUsage()
   False=StaticSwitchParameter(True=VertexColor(), False=Constant3Vector()) {group=Debug, parameterName=QT_ShowVertexColor, sortPriority=255}
  ) {group=Debug, parameterName=QT_ShowVinesAnimParameter, sortPriority=255}
  n6 = StaticSwitchParameter(
   True=NamedRerouteUsage()
   False=StaticSwitchParameter(
    True=NamedRerouteUsage()
    False=StaticSwitchParameter(True=@n4, False=@n5) {group=Debug, parameterName=QT_ShowThornsLegsGradient, sortPriority=255}
   ) {group=Debug, parameterName=QT_ShowThornsLegsColor, sortPriority=255}
  ) {group=Debug, parameterName=QT_ShowThornsLegsMask, sortPriority=255}
  n7 = Multiply(
   A=ScalarParameter() {defaultValue=1, group=Debug, parameterName=Debug Value Multiplier}
   B=StaticSwitchParameter(
    True=NamedRerouteUsage()
    False=StaticSwitchParameter(True=@n3, False=@n6) {group=Debug, parameterName=QT_ShowFresnelMasks, sortPriority=255}
   ) {group=Debug, parameterName=QT_ShowGradientMasks, sortPriority=255}
  )
  FunctionOutput(None=StaticSwitchParameter(
    True=SetMaterialAttributes!(MaterialAttributes=@n1, BaseColor=@n2, EmissiveColor=@n7) {attributeSetTypes [#1, #2]}
    False=@n1
   ) {group=Debug, parameterName=ShowDebug, sortPriority=1})
  n8 = VectorParameter() {group=Debug, parameterName=ThornsColor, defaultValue=<rgba 1 0.0156 0.0002 1>}
  n9 = FunctionInput() {inputName=InGradients, inputType=FunctionInput_Vector2, sortPriority=1}
  n10 = MaterialFunctionCall(Float2=NamedRerouteDeclaration(None=Ceil(None=@n9)) {name=ceilGradients, nodeColor=<rgba 0.4297 0 1 1>}) {materialFunction=&BreakOutFloat2Components}
  n11 = VectorParameter() {group=Debug, parameterName=LegsColor, defaultValue=<rgba 0.1359 1 0 1>}
  NamedRerouteDeclaration(None=Add(A=Multiply(A=@n8.RGB, B=@n10.R), B=Multiply(A=@n11.RGB, B=@n10.G))) {name=colorGradients, nodeColor=<rgba 1 0.5391 0 1>}
  n12 = FunctionInput() {inputName=InFresnelMask, inputType=FunctionInput_Scalar, sortPriority=2}
  n13 = VectorParameter() {group=Debug, parameterName=ThornsFresnel, defaultValue=<rgba 1 0.0156 0.0002 1>}
  n14 = MaterialFunctionCall(Float2=FunctionInput() {inputName=InFresnelAmounts, inputType=FunctionInput_Vector2, sortPriority=3}) {materialFunction=&BreakOutFloat2Components}
  n15 = VectorParameter() {group=Debug, parameterName=LegsFresnel, defaultValue=<rgba 0.1359 1 0 1>}
  n16 = Add(A=Multiply(A=@n13.RGB, B=@n14.R), B=Multiply(A=@n15.RGB, B=@n14.G))
  n17 = MaterialFunctionCall(Float2=NamedRerouteUsage()) {materialFunction=&BreakOutFloat2Components}
  n18 = Multiply(A=@n13.RGB, B=@n17.R)
  n19 = LinearInterpolate(A=@n18, B=Multiply(A=@n18, B=@n12), Alpha=@n14.R)
  n20 = Multiply(A=@n15.RGB, B=@n17.G)
  n21 = LinearInterpolate(A=@n20, B=Multiply(A=@n20, B=@n12), Alpha=@n14.G)
  NamedRerouteDeclaration(None=StaticSwitchParameter(True=Multiply(A=@n12, B=@n16), False=Add(A=@n19, B=@n21)) {group=C_AtlasColorBlend, parameterName=LK_ThornsInheritFromLegs, sortPriority=255}) {name=fresnelOutMasks, nodeColor=<rgba 1 0 0.8203 1>}
  NamedRerouteDeclaration(None=LinearInterpolate(
    A=VectorParameter() {group=Debug, parameterName=TLMask Min, defaultValue=<rgba 0 1 0 1>}.RGB
    B=VectorParameter() {group=Debug, parameterName=TLMask Max, defaultValue=<rgba 1 0 0 1>}.RGB
    Alpha=FunctionInput() {inputName=InThornsLegsMask, inputType=FunctionInput_Scalar, sortPriority=4}
   )) {name=thornsLegsMaskColor, nodeColor=<rgba 1 0 0.3281 1>}
  NamedRerouteDeclaration(None=FunctionInput() {inputName=InThornsLegsColor, sortPriority=5}) {name=thornsLegsInColor, nodeColor=<rgba 0.6406 1 0 1>}
  NamedRerouteDeclaration(None=FunctionInput() {inputName=InVinesAnimCurveTime, inputType=FunctionInput_Scalar, sortPriority=7}) {name=vinesAnimParam, nodeColor=<rgba 0.3594 1 0 1>}
  n22 = NamedRerouteDeclaration(None=FunctionInput() {inputName=InThornsLegsGradient, inputType=FunctionInput_Scalar, sortPriority=6}) {name=thornsLegsGradient, nodeColor=<rgba 0 1 0.4844 1>}
  n23 = ScalarParameter() {group=Debug, parameterName=TLInputMin}
  n24 = ScalarParameter() {defaultValue=1, group=Debug, parameterName=TLInputMax}
  NamedRerouteDeclaration(None=LinearInterpolate(
    A=VectorParameter() {group=Debug, parameterName=TLMask Min, defaultValue=<rgba 0 1 0 1>}.RGB
    B=VectorParameter() {group=Debug, parameterName=TLMask Max, defaultValue=<rgba 1 0 0 1>}.RGB
    Alpha=MaterialFunctionCall!(Input=@n22, Input Low=@n23, Input High=@n24) {materialFunction=/Engine/Functions/Engine_MaterialFunctions03/Math/RemapValueRangeNormalized}.Result
   )) {name=thornsLegsGradientCol, nodeColor=<rgba 1 0.1641 0 1>}
 ]}
