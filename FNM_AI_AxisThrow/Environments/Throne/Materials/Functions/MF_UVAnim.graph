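# Material function that pans UVs over time at an input speed, optionally varying that speed by mesh vertex normals.
# Time times speed is wrapped with Frac and added to the UV input, which previews as the default texture coordinates.
# Local-space vertex normals pass through a VertexInterpolator and FlattenNormal, then XY is multiplied by Z and scaled down.
# A static switch enables this offset, blended into the speed by NormalsAmount.
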MaterialFunction /FNM_AI_AxisThrow/Environments/Throne/Materials/Functions/MF_UVAnim
graph {nodes [
  n1 = FunctionInput() {bUsePreviewValueAsDefault=true, inputName=UVSpeed, inputType=FunctionInput_Vector2, sortPriority=1, previewValue=<1 1 0 1>}
  n2 = NamedRerouteDeclaration(None=@n1) {name=uvSpeed, nodeColor=<rgba 1 0 0.6094 1>}
  n3 = Add(A=@n2, B=NamedRerouteUsage())
  n4 = FunctionInput() {bUsePreviewValueAsDefault=true, inputName=NormalsAmount, inputType=FunctionInput_Scalar, sortPriority=4, previewValue=<1 1 0 1>}
  n5 = Multiply(A=Time(), B=StaticSwitch(True=LinearInterpolate(A=@n2, B=@n3, Alpha=@n4), False=NamedRerouteUsage(), Value=NamedRerouteUsage()))
  n6 = TextureCoordinate()
  FunctionOutput(None=Add(A=Frac(None=@n5), B=FunctionInput(Preview=@n6) {bUsePreviewValueAsDefault=true, inputName=UV, inputType=FunctionInput_Vector2}))
  n7 = VertexNormalWS()
  n8 = FunctionInput() {bUsePreviewValueAsDefault=true, inputName=FlattenNormals, inputType=FunctionInput_Scalar, sortPriority=3, previewValue=<0 1 0 1>}
  n9 = MaterialFunctionCall(
   Normal=VertexInterpolator(VS=Transform(None=@n7) {transformSourceType=TRANSFORMSOURCE_World, transformType=TRANSFORM_Local}).PS
   Flatness=@n8
  ) {materialFunction=&FlattenNormal}
  n10 = ComponentMask(None=@n9.Result) {g=true, r=true}
  n11 = ComponentMask(None=@n9.Result) {b=true}
  NamedRerouteDeclaration(None=Multiply(A=Multiply(A=@n10, B=@n11)) {constB=0.01}) {name=tnorm, nodeColor=<rgba 1 0.2812 0 1>}
  n12 = FunctionInput(Preview=StaticBool()) {bUsePreviewValueAsDefault=true, inputName=UseVtxNormals, inputType=FunctionInput_StaticBool, sortPriority=2, previewValue=<1 1 0 1>}
  NamedRerouteDeclaration(None=@n12) {name=useNormals, nodeColor=<rgba 0.2891 1 0 1>}
 ]}
