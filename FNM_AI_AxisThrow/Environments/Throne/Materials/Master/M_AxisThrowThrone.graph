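# Master material for the AxisThrow throne environment, with static switches for vine, thorn, fresnel and Lumpkin glow features.
# Up to two texture layers feed curve-atlas blending and emissive adjustments, or GigaMix replaces that stack.
# Normals become world space with backface correction, enabling optional height-based blending into landscape runtime virtual textures.
# MF_QTAnim vine animation, destructible damage and a debug function finish the chain, with material proxy replacements along the way.
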Material /FNM_AI_AxisThrow/Environments/Throne/Materials/Master/M_AxisThrowThrone
graph {
 properties {
  bAutomaticallySetUsageInEditor=false
  bTangentSpaceNormal=false
  bUseMaterialAttributes=true
  bUsedWithInstancedStaticMeshes=true
  bUsedWithNanite=true
  bUsedWithSplineMeshes=true
  ditheredLODTransition=true
  pixelDepthOffsetMode=PDOM_Legacy
  shadingModel=MSM_FromMaterialExpression
  usedShadingModels=Default Lit | Subsurface | Clear Coat | Two Sided Foliage
 }
 nodes [
  n1 = StaticSwitchParameter(
   True=MaterialFunctionCall() {materialFunction=/Game/Global/Base4Layers_2022/MaterialFunctions/MF_GigaMix_2023}.Result
   False=NamedRerouteUsage()
  ) {group=Options, parameterName=Use GigaMix}
  n2 = GetMaterialAttributes(None=@n1) {attributeGetTypes [#1]}
  n3 = NamedRerouteUsage()
  n4 = StaticSwitchParameter(
   True=SetMaterialAttributes(MaterialAttributes=@n2.MaterialAttributes, Emissive Color=Multiply(A=@n2.Emissive Color, B=@n3)) {attributeSetTypes [#1]}
   False=@n1
  ) {group=Options, parameterName=UseLumpkinInnerFalloffGlow}
  Composite!(None=@n4) {desc=Transform to World in case we will use RVT after this point, subgraphName=Tangent to World}
  n5 = NamedRerouteUsage()
  n6 = NamedRerouteUsage()
  n7 = NamedRerouteUsage()
  Composite!(InMA=@n5, thornsMask=@n6, legsMask=@n7) {subgraphName=EmissiveFresnelNodes}
  n8 = StaticBool() {value=true}
  n9 = StaticBool()
  MaterialExpressionReroute_3 = Reroute(None=@n4)
  n10 = GetMaterialAttributes(None=@MaterialExpressionReroute_3.Input) {attributeGetTypes [#2]}
  n11 = Transform(None=@n10.Normal)
  n12 = LinearInterpolate(A=Multiply(A=@n11) {constB=-1}, B=@n11, Alpha=Saturate(None=TwoSidedSign())) {desc="Fixes normals\r\non backfaces"}
  n13 = SetMaterialAttributes!(
   MaterialAttributes=@n10.MaterialAttributes
   Normal=StaticSwitchParameter(True=MaterialProxyReplace(Realtime=@n12, MaterialProxy=@n10.Normal), False=@n12) {parameterName=Force TangentSpaceNormalsFor HLOD ProxyReplace}
  ) {attributeSetTypes [#2]}
  MaterialExpressionReroute_2 = Reroute(None=@n13)
  n14 = NamedRerouteDeclaration(None=@MaterialExpressionReroute_2.Output) {name=Pre RVT, nodeColor=<rgba 1 0.7734 0 1>}
  n15 = VertexNormalWS()
  n16 = MaterialFunctionCall(Vector=AppendVector(A=ComponentMask(None=@n15) {g=true, r=true}, B=Constant()), Default=Constant3Vector() {constant=<rgba 0 0 1 0>}) {materialFunction=&SafeNormalize}
  n17 = WorldPosition()
  n18 = RuntimeVirtualTextureSampleParameter() {
   group=E_RVT
   materialType=WorldHeight
   parameterName=RVTHeight
   sortPriority=1
   virtualTexture=/Game/Athena/Apollo/Maps/Landscape/VT/RVT_Apollo_Landscape_Height
  }
  n19 = Subtract(
   A=Add(A=ScalarParameter() {defaultValue=-10, group=E_RVT, parameterName=Offset RVT WP, sortPriority=4}, B=ComponentMask(None=@n17.XYZ) {b=true})
   B=NamedRerouteDeclaration(None=@n18.WorldHeight) {name=RVT Height as is, nodeColor=<rgba 0 0.4141 1 1>}
  )
  n20 = RuntimeVirtualTextureSampleParameter!(
   UVs
   World Position=Subtract(A=WorldPosition() {worldPositionShaderOffset=WPT_ExcludeAllShaderOffsets}.XYZ, B=Multiply(A=@n16.Result, B=@n19))
   Mip Level=Constant()
  ) {
   bEnableFeedback=false
   desc=Use mip 0 but don't request feedback. Should give mip requested by landscape.
   group=E_RVT
   materialType=BaseColor_Normal_Specular_YCoCg
   mipValueMode=RVTMVM_MipLevel
   parameterName=RVT
   sortPriority=1
   virtualTexture=/Game/Athena/Apollo/Maps/Landscape/VT/RVT_Apollo_Landscape
  }
  n21 = SetMaterialAttributes!(
   MaterialAttributes
   Base Color=NamedRerouteDeclaration(None=@n20.BaseColor) {name=RVT Base Color, nodeColor=<rgba 0 1 0.6016 1>}
   Specular=NamedRerouteDeclaration(None=@n20.Specular) {name=RVT Specular, nodeColor=<rgba 0.5703 0 1 1>}
   Roughness=NamedRerouteDeclaration(None=@n20.Roughness) {name=RVT Rough, nodeColor=<rgba 0 1 0.7422 1>}
   Emissive Color=Constant()
   Normal=NamedRerouteDeclaration(None=@n20.Normal) {name=RVT N, nodeColor=<rgba 0 0.6953 1 1>}
  ) {attributeSetTypes [
    #3
    #4
    #5
    #1
    #2
   ]}
  n22 = MaterialProxyReplace(Realtime=BlendMaterialAttributes(A=@n21, B=@n14, Alpha=NamedRerouteUsage()) {vertexAttributeBlendType=UseA}, MaterialProxy=@n14)
  n23 = MaterialFunctionCall!(
   InMA=StaticSwitchParameter(
    True=MaterialFunctionCall(NoRVT=@n14, RVT=@n22) {materialFunction=&FN_RVT_Switch_Material}.Result
    False=NamedRerouteUsage()
   ) {group=E_RVT, parameterName=Use_RVT Blending, sortPriority=-1}
   UseQTAnim=NamedRerouteDeclaration(None=StaticSwitchParameter(True=@n8, False=@n9) {group=Options, parameterName=UseVineAnimation}) {name=useVineAnimBool, nodeColor=<rgba 0.1484 1 0 1>}
   AdditiveWPO
  ) {materialFunction=/FNM_AI_AxisThrow/Environments/Throne/Materials/Functions/MF_QTAnim}
  n24 = StaticBool() {value=true}
  n25 = StaticBool()
  MaterialFunctionCall!(
   InMA=MaterialFunctionCall(
    Material_In=@n23.outMA
    UseDestructibleDamage=StaticSwitchParameter(True=@n24, False=@n25) {defaultValue=true, group=Options, parameterName=UseDestructibleDamage}
    UseCrackedDamageTextures=@n25
    UseAdditiveWPO=NamedRerouteUsage()
   ) {materialFunction=&MF_FN_DestructibleDamage}.Material_Out
   InGradients=AppendVector(A=NamedRerouteUsage(), B=NamedRerouteUsage())
   InFresnelMask=NamedRerouteUsage()
   InFresnelAmounts=NamedRerouteUsage()
   InThornsLegsMask=NamedRerouteUsage()
   InThornsLegsColor=NamedRerouteUsage()
   InThornsLegsGradient=NamedRerouteUsage()
   InVinesAnimCurveTime=NamedRerouteUsage()
  ) {materialFunction=/FNM_AI_AxisThrow/Environments/Throne/Materials/Functions/MF_QT_Debug}
  NamedRerouteDeclaration(None=TextureObjectParameter() {group=A_Layer_1, parameterName=L1_Albedo, sortPriority=16}) {name=L1Albedo, nodeColor=<rgba 1 0.0234 0 1>}
  n26 = GetMaterialAttributes(None=NamedRerouteUsage()) {attributeGetTypes [#3]}
  n27 = VectorParameter() {group=Vine Color Variation, parameterName=Vine Color Variation Mask R, sortPriority=3, defaultValue=<rgba 0.2781 0.038 0.0077 1>}
  n28 = TextureSampleParameter2D(UVs=Divide(
    A=TextureCoordinate()
    B=ScalarParameter() {defaultValue=0.3, group=Vine Color Variation, parameterName=Color Variation Mask Tiling, sortPriority=2}
   )) {
   group=Vine Color Variation
   parameterName=Color Variation Mask Texture
   samplerType=SAMPLERTYPE_Masks
   sortPriority=1
   texture=/Game/Environments/Helios/Props/Holiday/Holiday_LumpkinKing_A/Textures/Tiling/T_AxisThrow_Vine_Tiling_Mask
  }
  n29 = Multiply(
   A=@n28.R
   B=ScalarParameter() {defaultValue=0.25, group=Vine Color Variation, parameterName=Vine Color Variation R Intensity, sortPriority=4}
  )
  n30 = VectorParameter() {group=Vine Color Variation, parameterName=Vine Color Variation Mask G, sortPriority=5, defaultValue=<rgba 0.1085 0.4531 0.0765 1>}
  n31 = ScalarParameter() {defaultValue=0.1, group=Vine Color Variation, parameterName=Vine Color Variation G Intensity, sortPriority=6}
  n32 = VectorParameter() {group=Vine Color Variation, parameterName=Vine Color Variation Mask B, sortPriority=7, defaultValue=<rgba 0.2058 0.3698 0.005 1>}
  n33 = LinearInterpolate(
   A=LinearInterpolate(A=LinearInterpolate(A=@n26.Base Color, B=@n27.RGB, Alpha=@n29), B=@n30.RGB, Alpha=Multiply(A=@n28.G, B=@n31))
   B=@n32.RGB
   Alpha=Multiply(A=@n28.B, B=ScalarParameter() {defaultValue=0.2, group=Vine Color Variation, parameterName=Vine Color Variation B Intensity, sortPriority=8})
  )
  NamedRerouteDeclaration(None=StaticSwitchParameter(
    True=SetMaterialAttributes!(MaterialAttributes=@n26.MaterialAttributes, Base Color=@n33) {attributeSetTypes [#3]}
    False=NamedRerouteUsage()
   ) {group=Vine Color Variation, parameterName=Use Color Varitation Mask}) {name=L1_MA, nodeColor=<rgba 0 0.8125 1 1>}
  n34 = CurveAtlasRowParameter(CurveTime=NamedRerouteUsage()) {
   atlas=/FNM_AI_AxisThrow/Environments/Throne/Materials/Functions/Curves/T_QTCurveAtlas
   curve=/FNM_AI_AxisThrow/Environments/Throne/Materials/Functions/Curves/Albedo_Curve_01
   group=A_Layer_1
   parameterName=L1 EmissiveGradient
  }
  n35 = VectorParameter() {group=A_Layer_1, parameterName=L1 Emissive Color}
  n36 = ScalarParameter() {defaultValue=1, group=A_Layer_1, parameterName=L1 EmissiveValue}
  n37 = ScalarParameter() {defaultValue=1, group=A_Layer_1, parameterName=L1 EmissiveLumen}
  NamedRerouteDeclaration(None=Multiply(
    A=StaticSwitchParameter(True=@n34, False=@n35.RGB) {desc=Use AOSR instead, group=A_Layer_1, parameterName=L1_UseEmissiveGradient, sortPriority=255}
    B=RayTracingQualitySwitch(Normal=@n36, RayTraced=@n37)
   )) {name=L1EmColor, nodeColor=<rgba 0.9453 1 0 1>}
  NamedRerouteDeclaration(None=VectorParameter() {group=A_Layer_2, parameterName=L2 Emissive Color}.RGB) {name=L2EmColor, nodeColor=<rgba 0.7344 0 1 1>}
  NamedRerouteDeclaration(None=TextureObjectParameter() {group=A_Layer_2, parameterName=L2_AOER, sortPriority=16}) {name=L2AOER, nodeColor=<rgba 1 0.0234 0 1>}
  NamedRerouteDeclaration(None=TextureObjectParameter() {group=A_Layer_2, parameterName=L2_AOSR, sortPriority=16}) {name=L2AOSR, nodeColor=<rgba 1 0.1978 0.8982 1>}
  n38 = StaticBool() {value=true}
  n39 = StaticBool()
  NamedRerouteDeclaration(None=StaticSwitchParameter(True=@n38, False=@n39) {defaultValue=true, desc=Use AOSR instead, group=A_Layer_2, parameterName=L2_UseAOER, sortPriority=255}) {name=L2UseAOER, nodeColor=<rgba 0.3125 0 1 1>}
  NamedRerouteDeclaration(None=StaticSwitchParameter(True=TextureCoordinate() {coordinateIndex=1}, False=TextureCoordinate()) {desc=Use UV0 otherwise..., group=A_Layer_1, parameterName=L1_UseUV1, sortPriority=0}) {name=L1UVIndex, nodeColor=<rgba 0.7344 1 0 1>}
  n40 = ScalarParameter() {group=A_Layer_1, parameterName=L1 U speed, sortPriority=64}
  n41 = ScalarParameter() {group=A_Layer_1, parameterName=L1 V speed, sortPriority=64}
  n42 = StaticBool()
  n43 = StaticSwitchParameter(True=StaticBool() {value=true}, False=@n42) {group=A_Layer_1, parameterName=L1_AnimUseNormals, sortPriority=255}
  NamedRerouteDeclaration(None=MaterialFunctionCall!(
    UV=NamedRerouteUsage()
    UVSpeed=AppendVector(A=@n40, B=@n41)
    UseVtxNormals=StaticSwitchParameter(True=@n43, False=@n42) {group=A_Layer_1, parameterName=L1_UseUVAnim, sortPriority=255}
    FlattenNormals=ScalarParameter() {group=A_Layer_1, parameterName=L1 AnimNormals Flatten, sortPriority=50}
    NormalsAmount=ScalarParameter() {group=A_Layer_1, parameterName=L1 AnimNormals Amount, sortPriority=50}
   ) {materialFunction=/FNM_AI_AxisThrow/Environments/Throne/Materials/Functions/MF_UVAnim}.Result) {name=L1animatedUVs, nodeColor=<rgba 0.9453 0 1 1>}
  NamedRerouteDeclaration(None=StaticSwitchParameter(True=@n38, False=@n39) {defaultValue=true, desc=Use AOSR instead, group=A_Layer_2, parameterName=L2_UseAOEROnly, sortPriority=255}) {name=L2UseOnlyAOER, nodeColor=<rgba 0.4531 1 0 1>}
  n44 = GetMaterialAttributes(None=NamedRerouteUsage()) {attributeGetTypes [#5, #6, #1]}
  n45 = GetMaterialAttributes(None=NamedRerouteUsage()) {attributeGetTypes [#5, #6, #1]}
  n46 = Multiply(A=@n44.Roughness, B=@n45.Roughness)
  n47 = Multiply(A=@n44.Ambient Occlusion, B=@n45.Ambient Occlusion)
  n48 = Add(A=@n44.Emissive Color, B=@n45.Emissive Color)
  NamedRerouteDeclaration(None=StaticSwitchParameter(
    True=SetMaterialAttributes!(MaterialAttributes=@n44.MaterialAttributes, Roughness=@n46, Ambient Occlusion=@n47, Emissive Color=@n48) {attributeSetTypes [#5, #6, #1]}
    False=NamedRerouteUsage()
   ) {group=Options, parameterName=Use Second Layer}) {name=blendedMA, nodeColor=<rgba 0 0.2969 1 1>}
  n49 = NamedRerouteUsage()
  n50 = NamedRerouteUsage()
  n51 = NamedRerouteUsage()
  n52 = MaterialFunctionCall!(
   MA=@n49
   EmissiveColor=StaticSwitchParameter(True=StaticSwitchParameter(True=@n50, False=@n51) {group=C_AtlasColorBlend, parameterName=Use L1 EmissiveColor}, False=NamedRerouteUsage()) {group=Options, parameterName=Use Second Layer}
  ) {materialFunction=/FNM_AI_AxisThrow/Environments/Throne/Materials/Functions/MF_QTAtlasBlend}
  n53 = StaticSwitchParameter(
   True=NamedRerouteUsage()
   False=NamedRerouteDeclaration(None=StaticSwitchParameter(True=@n52.Result, False=@n49) {group=Options, parameterName=Use CurveBlends}) {name=curveBlendsMA, nodeColor=<rgba 1 0 0.2812 1>}
  ) {group=Options, parameterName=UseEmissiveFresnel, sortPriority=50}
  n54 = GetMaterialAttributes(None=@n53) {attributeGetTypes [#1]}
  n55 = NamedRerouteUsage()
  n56 = ScalarParameter() {defaultValue=1, group=D_ThornsEmissive, parameterName=ThornsGradientContrast}
  n57 = TextureSample(
   UVs=NamedRerouteUsage()
   Tex=StaticSwitchParameter(True=NamedRerouteUsage(), False=NamedRerouteUsage()) {desc=Use AOSR instead, group=A_Layer_1, parameterName=L1_UseAOER, sortPriority=255}
  )
  n58 = NamedRerouteUsage()
  n59 = Add(A=@n54.Emissive Color, B=Multiply(A=Power(Base=@n55, Exp=@n56), B=Multiply(A=@n57.G, B=@n58)))
  n60 = ScalarParameter() {defaultValue=1, group=D_ThornsEmissive, parameterName=ThornsFixAmount}
  n61 = StaticSwitchParameter(
   True=SetMaterialAttributes!(MaterialAttributes=@n54.MaterialAttributes, Emissive Color=LinearInterpolate(A=@n54.Emissive Color, B=@n59, Alpha=@n60)) {attributeSetTypes [#1]}
   False=@n53
  ) {desc=Use AOSR instead, group=Options, parameterName=UseThornsEmissiveFix, sortPriority=255}
  NamedRerouteDeclaration(None=MaterialFunctionCall!(InMA=StaticSwitchParameter(True=@n61, False=@n53) {desc=Use AOSR instead, group=A_Layer_1, parameterName=L1_UseAOER, sortPriority=255}) {materialFunction=/FNM_AI_AxisThrow/Environments/Throne/Materials/Functions/MF_EmissiveAdjustmentsMA}.Result) {name=Yumebau, nodeColor=<rgba 0 0.7656 1 1>}
  n62 = Divide(A=PixelDepth().R, B=ScalarParameter() {defaultValue=1000, group=A_Layer_1, parameterName=L1_DetailMaxDistance, sortPriority=100})
  n63 = VectorParameter() {
   group=A_Layer_1
   parameterName=L1_DetailAlpha
   channelNames {a="", b=Normal Alpha, g=AOSR Alpha, r=Albedo Alpha}
   defaultValue=<rgba 1 1 1 1>
  }
  n64 = MaterialFunctionCall(X=@n63.Albedo Alpha, Y=@n63.Albedo Alpha, Z=@n63.Albedo Alpha, A=@n63.Albedo Alpha) {materialFunction=&MakeFloat4}
  n65 = Multiply(A=OneMinus(None=Saturate(None=@n62)), B=@n64.Result)
  NamedRerouteDeclaration(None=StaticSwitchParameter(
    True=StaticSwitchParameter(True=@n65, False=@n64.Result) {defaultValue=true, group=A_Layer_1, parameterName=L1_FadeDetailMaps, sortPriority=255}
    False=@n64.Result
   ) {group=A_Layer_1, parameterName=L1_UseDetailMaps, sortPriority=255}) {name=L1TriAlpha, nodeColor=<rgba 1 0 0.6562 1>}
  NamedRerouteDeclaration(None=@n52.ThornsParam) {name=thornsGradient, nodeColor=<rgba 1 0.2344 0 1>}
  NamedRerouteDeclaration(None=@n52.LegsParam) {name=legsGradient, nodeColor=<rgba 0 0.7656 1 1>}
  n66 = NamedRerouteUsage()
  n67 = StaticBool() {value=true}
  n68 = StaticBool()
  n69 = NamedRerouteDeclaration(None=MaterialFunctionCall(
    Invert Fresnel=StaticSwitchParameter(True=@n67, False=@n68) {group=Fresnel, parameterName=InvertFresnel, sortPriority=50}
    Power=ScalarParameter() {defaultValue=5, group=Fresnel, parameterName=Fresnel Exponent, sortPriority=50}
   ) {materialFunction=&Fresnel_Function}.Result) {name=gradientsFresnel, nodeColor=<rgba 0 0.0156 1 1>}
  n70 = Constant() {r=1}
  MaterialExpressionReroute_1 = Reroute(None=@n6)
  n71 = StaticSwitchParameter(True=@n70, False=@MaterialExpressionReroute_1.thornsMask) {group=C_AtlasColorBlend, parameterName=LK_ThornsInheritFromLegs, sortPriority=255}
  n72 = ScalarParameter() {defaultValue=1, group=Fresnel, parameterName=Fresnel Thorns, sortPriority=50}
  n73 = LinearInterpolate(
   A=NamedRerouteUsage()
   B=Multiply(A=@n66, B=@n69)
   Alpha=Multiply(A=NamedRerouteDeclaration(None=@n71) {name=thornsCeilMask, nodeColor=<rgba 0 0.5078 1 1>}, B=@n72)
  )
  MaterialExpressionReroute_5 = Reroute(None=@n7)
  n74 = StaticSwitchParameter(True=@n70, False=@MaterialExpressionReroute_5.legsMask) {group=C_AtlasColorBlend, parameterName=LK_ThornsInheritFromLegs, sortPriority=255}
  n75 = ScalarParameter() {defaultValue=1, group=Fresnel, parameterName=Fresnel Legs, sortPriority=50}
  n76 = LinearInterpolate(
   A=NamedRerouteUsage()
   B=Multiply(A=@n66, B=@n69)
   Alpha=Multiply(A=NamedRerouteDeclaration(None=@n74) {name=legsCeilMask, nodeColor=<rgba 0.8281 1 0 1>}, B=@n75)
  )
  n77 = NamedRerouteUsage()
  n78 = NamedRerouteUsage()
  n79 = LinearInterpolate(B=NamedRerouteUsage(), Alpha=ScalarParameter() {defaultValue=1, group=Fresnel, parameterName=Fresnel Amount, sortPriority=50}) {constA=1}
  MaterialExpressionReroute_12 = Reroute(None=@n5)
  n80 = GetMaterialAttributes(None=@MaterialExpressionReroute_12.InMA) {attributeGetTypes [#1]}
  n81 = SetMaterialAttributes(
   MaterialAttributes=@n80.MaterialAttributes
   Emissive Color=StaticSwitchParameter(True=LinearInterpolate(A=@n73, B=@n76, Alpha=@n77), False=Multiply(A=@n78, B=@n79)) {group=Options, parameterName=Use CurveBlends}
  ) {attributeSetTypes [#1]}
  MaterialExpressionReroute_11 = Reroute(None=@n81)
  NamedRerouteDeclaration(None=@MaterialExpressionReroute_11.OutMA) {name=emissiveFresnelMA, nodeColor=<rgba 0.7578 0 1 1>}
  NamedRerouteDeclaration(None=@n80.Emissive Color) {name=inEmissive, nodeColor=<rgba 0.3359 1 0 1>}
  n82 = AppendVector(
   A=ScalarParameter() {defaultValue=1, group=D_ThornsEmissive, parameterName=Thorns U Scale}
   B=ScalarParameter() {defaultValue=1, group=D_ThornsEmissive, parameterName=Thorns V Scale}
  )
  n83 = StaticSwitchParameter(True=TextureCoordinate() {coordinateIndex=1}, False=TextureCoordinate()) {desc=Use AOSR instead, group=D_ThornsEmissive, parameterName=ThornsFix_UseUV1, sortPriority=255}
  n84 = ScalarParameter() {group=A_Layer_1, parameterName=L1 U speed, sortPriority=64}
  n85 = ScalarParameter() {group=A_Layer_1, parameterName=L1 V speed, sortPriority=64}
  n86 = StaticBool()
  n87 = StaticSwitchParameter(True=StaticBool() {value=true}, False=@n86) {group=A_Layer_1, parameterName=L1_AnimUseNormals, sortPriority=255}
  NamedRerouteDeclaration(None=MaterialFunctionCall(
    UV=Multiply(A=@n82, B=@n83)
    UVSpeed=AppendVector(A=@n84, B=@n85)
    UseVtxNormals=StaticSwitchParameter(True=@n87, False=@n86) {group=A_Layer_1, parameterName=L1_UseUVAnim, sortPriority=255}
    FlattenNormals=ScalarParameter() {group=A_Layer_1, parameterName=L1 AnimNormals Flatten, sortPriority=50}
    NormalsAmount=ScalarParameter() {group=A_Layer_1, parameterName=L1 AnimNormals Amount, sortPriority=50}
   ) {materialFunction=&MF_UVAnim}.Result) {name=thornsUVAnim, nodeColor=<rgba 0.4297 1 0 1>}
  NamedRerouteDeclaration(None=@n52.ThornsCeil) {name=thornsCeilGradient, nodeColor=<rgba 0.9219 0 1 1>}
  NamedRerouteDeclaration(None=@n52.LegsCeil) {name=legsCeilGradient, nodeColor=<rgba 0 1 0.6016 1>}
  NamedRerouteDeclaration(None=AppendVector(A=@n72, B=@n75)) {name=fresnelAmounts, nodeColor=<rgba 0 1 0.1094 1>}
  n88 = StaticBool() {value=true}
  n89 = StaticBool()
  NamedRerouteDeclaration(None=StaticSwitchParameter(
    True=StaticSwitchParameter(True=@n88, False=@n89) {desc=Use UV0 otherwise..., group=A_Layer_1, parameterName=L1_UseDetailAOSR, sortPriority=255}
    False=@n89
   ) {group=A_Layer_1, parameterName=L1_UseDetailMaps, sortPriority=255}) {name=L1UseDetailAOSR, nodeColor=<rgba 0 0.9766 1 1>}
  n90 = ScalarParameter() {defaultValue=2, group=E_InnerGlowFalloff, parameterName=Lumpkin_InnerGlowFalloff}
  n91 = Fresnel(ExponentIn=@n90) {baseReflectFraction=0}
  NamedRerouteDeclaration(None=LinearInterpolate(
    A=ScalarParameter() {defaultValue=2, group=E_InnerGlowFalloff, parameterName=Lumpkin_InnerglowBrightness_Max}
    B=ScalarParameter() {group=E_InnerGlowFalloff, parameterName=Lumpkin_InnerglowBrightness_Min}
    Alpha=Saturate(None=@n91)
   )) {name=Lumpkin_InnerglowFalloff, nodeColor=<rgba 0 1 0.1797 1>}
  n92 = NamedRerouteUsage()
  MaterialExpressionReroute_0 = Reroute(None=@n92)
  NamedRerouteDeclaration(None=@MaterialExpressionReroute_0.fresnelGradient) {name=fresnelGradient_S, nodeColor=<rgba 1 0.6328 0 1>}
  n93 = ScalarParameter() {
   defaultValue=150
   desc="Used to control how far colors will blend with the rvt, as well sets blending distance of normals for Up looking surfaces"
   group=E_RVT
   parameterName=Blend RVT Distance
   sortPriority=2
  }
  n94 = Divide(A=@n19, B=@n93)
  n95 = ScalarParameter() {
   defaultValue=1
   desc="Used to control how far colors will blend with the rvt, as well sets blending distance of normals for Up looking surfaces"
   group=E_RVT
   parameterName=Blend RVT Falloff
   sortPriority=2
  }
  NamedRerouteDeclaration(None=Saturate(None=Power(Base=@n94, Exp=@n95))) {name=RVT Blending, nodeColor=<rgba 0.9922 1 0 1>}
  n96 = NamedRerouteUsage()
  MaterialExpressionReroute_7 = Reroute(None=@n96)
  NamedRerouteDeclaration(None=@MaterialExpressionReroute_7.fresnelAmount) {name=fresnelAmountV2, nodeColor=<rgba 0.3125 0 1 1>}
  n97 = TextureObjectParameter() {group=A_Layer_1, parameterName=L1_Normals, samplerType=SAMPLERTYPE_Normal, sortPriority=16, texture=/Engine/EngineMaterials/FlatNormal}
  NamedRerouteDeclaration(None=@n97) {name=L1Normal, nodeColor=<rgba 0 0.2969 1 1>}
  n98 = NamedRerouteUsage()
  n99 = NamedRerouteUsage()
  n100 = StaticBool() {value=true}
  n101 = StaticBool()
  n102 = VectorParameter() {desc=WorldSpace size, group=A_Layer_1, parameterName=L1_DetailSize, channelNames {a="", b=Z, g=Y, r=X}, defaultValue=<rgba 64 64 64 0>}
  n103 = StaticBool() {value=true}
  n104 = StaticBool()
  n105 = StaticBool() {value=true}
  n106 = StaticBool()
  n107 = StaticBool() {value=true}
  n108 = StaticBool()
  n109 = VectorParameter() {
   group=A_Layer_1
   parameterName=L1 UVAnim Amount
   channelNames {a="", b=Roughness Anim, g=Emissive-Specular, r=AO Anim}
   defaultValue=<rgba 1 1 1 0>
  }
  NamedRerouteDeclaration(None=MaterialFunctionCall!(
    UVs=NamedRerouteUsage()
    Tobj_Albedo=NamedRerouteUsage()
    Tobj_Normal=NamedRerouteUsage()
    Tobj_AOSR=StaticSwitchParameter(True=@n98, False=@n99) {desc=Use AOSR instead, group=A_Layer_1, parameterName=L1_UseAOER, sortPriority=255}
    TriplanarAlpha=NamedRerouteUsage()
    AllowTriplanar=StaticSwitchParameter(True=@n100, False=@n101) {group=A_Layer_1, parameterName=L1_UseDetailMaps, sortPriority=255}
    TriplanarSize=@n102
    Tobj_DetailNormal=NamedRerouteUsage()
    Tobj_DetailAOSR=NamedRerouteUsage()
    UseAOSRDetail=NamedRerouteUsage()
    UseUVDetailMaps
    UVDetailTileIndex
    UseAOER=StaticSwitchParameter(True=@n103, False=@n104) {desc=Use AOSR instead, group=A_Layer_1, parameterName=L1_UseAOER, sortPriority=255}
    UseSER=StaticSwitchParameter(True=@n105, False=@n106) {desc="Specular, Emissive, Roughness", group=A_Layer_1, parameterName=L1_UseSER, sortPriority=255}
    SpecularValue=ScalarParameter() {defaultValue=0.5, group=A_Layer_1, parameterName=L1 Specular, sortPriority=48}
    Emissive Color=NamedRerouteUsage()
    UseMasksOnly
    UseUVAnimation=StaticSwitchParameter(True=@n107, False=@n108) {group=A_Layer_1, parameterName=L1_UseUVAnim, sortPriority=255}
    AnimAOSRweights=@n109
    AnimatedUVs=NamedRerouteUsage()
    UseOpacityMask=StaticSwitchParameter(True=@n105, False=@n106) {desc=Use Opacity Mask for Fallback, group=A_Layer_1, parameterName=L1_UseOpacityMask, sortPriority=255}
    UseAlbedoAlpha=StaticSwitchParameter(True=@n105, False=@n106) {desc=Use Albedo Alpha as Opacity Mask, group=A_Layer_1, parameterName=L1_UseAlbedoAlpha, sortPriority=255}
    Tobj_OpacityMask=NamedRerouteUsage()
   ) {materialFunction=/FNM_AI_AxisThrow/Environments/Throne/Materials/Functions/MF_QTLayer}.Result) {name=L1Out, nodeColor=<rgba 0 0.6719 1 1>}
  NamedRerouteDeclaration(None=@n52.ThornsLegsMask) {name=thornsLegsMask, nodeColor=<rgba 0 1 0.9531 1>}
  NamedRerouteDeclaration(None=@n52.ThornsLegsColor) {name=thornsLegsColor, nodeColor=<rgba 1 0.9609 0 1>}
  NamedRerouteDeclaration(None=@n52.ThornsLegsParam) {name=thornsLegsGradient, nodeColor=<rgba 0.7109 0 1 1>}
  NamedRerouteDeclaration(None=@n52.LegsUVParams) {name=legsUVParams, nodeColor=<rgba 0 0.3906 1 1>}
  NamedRerouteDeclaration(None=@n23.CurveTime) {name=vinesAnimCurveTime, nodeColor=<rgba 0 0.4375 1 1>}
  NamedRerouteDeclaration(None=TextureObjectParameter() {group=A_Layer_1, parameterName=L1_OpacityMask, sortPriority=64, texture=/MeshModelingToolsetExp/Textures/DefaultTexture}) {name=L1OpacityMask, nodeColor=<rgba 0.2656 0 1 1>}
  NamedRerouteDeclaration(None=TextureObjectParameter() {group=A_Layer_1, parameterName=L1_AOSR, sortPriority=16, texture=&DefaultTexture}) {name=L1AOSR, nodeColor=<rgba 0.6172 1 0 1>}
  n110 = TextureObjectParameter() {group=A_Layer_1, parameterName=L1_DetailNormals, samplerType=SAMPLERTYPE_Normal, sortPriority=64, texture=&FlatNormal}
  NamedRerouteDeclaration(None=@n110) {name=L1DNormals, nodeColor=<rgba 1 0 0.9375 1>}
  NamedRerouteDeclaration(None=TextureObjectParameter() {group=A_Layer_1, parameterName=L1_DetailAOSR, sortPriority=64, texture=&DefaultTexture}) {name=L1DAOSR, nodeColor=<rgba 0 1 0.7422 1>}
  NamedRerouteDeclaration(None=TextureObjectParameter() {group=A_Layer_1, parameterName=L1_AOER, sortPriority=16, texture=&DefaultTexture}) {name=L1AOER, nodeColor=<rgba 0.2026 0.2461 1 1>}
  n111 = TextureCoordinate() {coordinateIndex=1}
  n112 = TextureCoordinate()
  n113 = TextureObject()
  n114 = NamedRerouteUsage()
  n115 = NamedRerouteUsage()
  NamedRerouteDeclaration(None=MaterialFunctionCall(
    UVs=StaticSwitchParameter(True=@n111, False=@n112) {desc=Use UV0 otherwise..., group=A_Layer_2, parameterName=L2_UseUV1, sortPriority=0}
    Tobj_Albedo=@n113
    Tobj_Normal=@n113
    Tobj_AOSR=StaticSwitchParameter(True=@n114, False=@n115) {defaultValue=true, desc=Use AOSR instead, group=A_Layer_2, parameterName=L2_UseAOER, sortPriority=255}
    Tobj_DetailNormal=@n113
    Tobj_DetailAOSR=@n113
    UseAOER=NamedRerouteUsage()
    SpecularValue=ScalarParameter() {defaultValue=0.5, group=A_Layer_2, parameterName=L2 Specular, sortPriority=48}
    Emissive Color=NamedRerouteUsage()
    UseMasksOnly=NamedRerouteUsage()
   ) {materialFunction=&MF_QTLayer}.Result) {name=L2_MA, nodeColor=<rgba 0.9643 0.1427 1 1>}
  PinBase() {reroutePins [Input {expression=&self:MaterialGraph_0.MaterialGraphNode_Composite_20.Tangent to World.MaterialExpressionReroute_0}]}
  PinBase!(Output) {reroutePins [Output {expression=&self:MaterialGraph_0.MaterialGraphNode_Composite_20.Tangent to World.MaterialExpressionReroute_1}]}
  PinBase!() {reroutePins [
    "For Buildings (MA)" {expression=null}
    "RVT_Height (S)" {expression=null}
   ]}
  PinBase() {reroutePins [Result {expression=null}]}
  PinBase() {reroutePins [Input {expression=&self:MaterialExpressionReroute_3}]}
  PinBase!(Output=@n13) {reroutePins [Output {expression=&self:MaterialExpressionReroute_2}]}
  PinBase!() {reroutePins [MaterialAttributes {expression=null}]}
  PinBase() {reroutePins [Result {expression=null}]}
  PinBase() {reroutePins [
    InMA {expression=&self:MaterialExpressionReroute_12}
    thornsMask {expression=&self:MaterialExpressionReroute_1}
    legsMask {expression=&self:MaterialExpressionReroute_5}
   ]}
  PinBase!(OutMA=@n81, fresnelGradient=@n92, fresnelAmount=@n96) {reroutePins [
    OutMA {expression=&self:MaterialExpressionReroute_11}
    fresnelGradient {expression=&self:MaterialExpressionReroute_0}
    fresnelAmount {expression=&self:MaterialExpressionReroute_7}
   ]}
  PinBase!() {reroutePins [InputPin {expression=null}]}
  PinBase() {reroutePins [Output {expression=null}]}
  n116 = Transform(None=Constant3Vector() {constant=<rgba 0 0 1 0>}) {transformSourceType=TRANSFORMSOURCE_Camera}
  n117 = Transform(None=Constant3Vector() {constant=<rgba 0 1 0 0>}) {transformSourceType=TRANSFORMSOURCE_Camera}
  n118 = ObjectPositionWS()
  n119 = Multiply(A=ScalarParameter() {defaultValue=1, group=E_InnerGlowFalloff, parameterName=Lumpkin_InnerGlowSize}, B=ObjectRadius())
  Power(
   Base=SphereMask(
    A=MaterialFunctionCall(Plane Normal Axis=@n116, Plane X Axis=@n117, Plane Center=@n118, UV Scale=@n119) {materialFunction=&VirtualPlaneCoordinates}.UVs
    B=Constant2Vector() {g=0.5, r=0.5}
   ) {attenuationRadius=1, hardnessPercent=0}
   Exp=@n90
  )
 ]
}
